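# Translucent unlit material for Niagara firework particles, previewed on a sphere mesh, rendering sparks as soft, flickering, hue-shiftable dots.
# World position offset combines sprite-style quad billboarding, vertex-color-varied radial expansion and further offsets, with named setups for fall, stretch and curl.
# Dynamic parameters drive quad scale, size, fall, curl and hue.
# Shading-path and quality switches choose between offset variants, with both vector-noise and cheap sine-based curl defined.
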Material /SparksEffects/Effects/Fireworks/Materials/M_Sparks_Firework_Mesh
graph {
 properties {
  bUsedWithNiagaraMeshParticles=true
  bUsedWithNiagaraSprites=true
  blendMode=BLEND_Translucent
  pixelDepthOffsetMode=PDOM_Legacy
  previewMesh=/SparksEffects/Effects/Fireworks/Meshes/SM_Sparks_Firework_Sphere
  shadingModel=MSM_Unlit
  twoSided=true
  usedShadingModels=Unlit
 }
 nodes [
  n1 = DynamicParameter() {defaultValue=<rgba 10 1 1 1>, paramNames [QuadScale,Size,SpriteFall,CurlNoise]}
  n2 = MaterialFunctionCall!(Center Location=WorldPosition().XYZ, 0-1 UVs, XY scale=@n1.QuadScale, OverridePivot, "Normals (Optional)") {materialFunction=&Sprite}
  n3 = WorldPosition()
  n4 = ObjectPositionWS()
  n5 = Multiply(A=Multiply(A=Subtract(A=@n3.XYZ, B=@n4), B=@n1.Size), B=Clamp(None=VertexColor()) {minDefault=0.95})
  n6 = NamedRerouteUsage()
  n7 = ShadingPathSwitch(Default=LinearInterpolate(B=NamedRerouteUsage(), Alpha=@n1.CurlNoise), Mobile=@n6) {desc=Migrated from FeatureLevelSwitch}
  Add(
   A=Add(A=Add(A=@n2.World position offset, B=@n5), B=NamedRerouteUsage())
   B=ShadingPathSwitch(Default=QualitySwitch(Default=@n7, Low=@n6, Medium=@n6), Forward=@n6, Mobile=@n6)
  )
  Constant()
  DynamicParameter() {parameterIndex=1, defaultValue=<rgba 0 1 1 1>, paramNames [HueShift,Param2,Param3,Param4]}
  n8 = DynamicParameter() {parameterIndex=1, defaultValue=<rgba 0 1 1 1>, paramNames [HueShift,Param2,Param3,Param4]}
  n9 = VertexColor()
  n10 = NamedRerouteUsage()
  n11 = NamedRerouteUsage()
  MaterialFunctionCall(Emissive=MaterialFunctionCall(Hue Shift Percentage=Multiply(A=@n8.HueShift, B=@n9.R), Texture=Multiply(A=@n10, B=@n11)) {materialFunction=&HueShift}.Result) {materialFunction=&MF_FX_EmissiveAdjustments}
  Multiply(A=Saturate(None=NamedRerouteUsage()), B=NamedRerouteUsage())
  n12 = Multiply(A=Time(), B=ScalarParameter() {defaultValue=1, parameterName=Flicker Speed})
  n13 = VertexColor()
  n14 = Multiply(A=Sine(None=Add(A=@n12, B=@n13)) {period=0.75}) {constB=1.5}
  n15 = Constant() {r=1}
  n16 = ScalarParameter() {defaultValue=0.5, parameterName=Flicker Amount}
  NamedRerouteDeclaration(None=Max(A=Desaturation(None=@n14, Fraction=@n15), B=OneMinus(None=@n16)) {constB=0.25}) {name=Flicker, nodeColor=<rgba 0 0 0 1>}
  n17 = ParticleColor()
  NamedRerouteDeclaration(None=MaterialFunctionCall(LDR Input=Clamp(None=@n17.RGB), HDR Tint=@n17.RGB) {materialFunction=&DeriveHDRfromLDR}.RemappedRange) {name=Color, nodeColor=<rgba 0 0 0 1>}
  n18 = VertexColor()
  n19 = Divide(A=Subtract(A=Add(A=@n17.A, B=@n18.R), B=OneMinus(None=@n17.A))) {constB=1}
  n20 = TextureCoordinate()
  n21 = Constant2Vector() {g=0.5, r=0.5}
  n22 = Constant() {r=0.5}
  n23 = Constant() {r=0.5}
  NamedRerouteDeclaration(None=Multiply(A=Saturate(None=@n19), B=SphereMask(A=@n20, B=@n21, Radius=@n22, Hardness=@n23))) {name=Alpha, nodeColor=<rgba 0 0 0 1>}
  n24 = Constant()
  n25 = AppendVector(A=@n24, B=@n24)
  n26 = DynamicParameter() {defaultValue=<rgba 10 1 1 1>, paramNames [QuadScale,Size,SpriteFall,CurlNoise]}
  n27 = Clamp(None=VertexColor()) {minDefault=0.75}
  n28 = ComponentMask(None=TextureCoordinate() {coordinateIndex=1}) {g=true}
  n29 = Saturate(None=Add(A=Multiply(A=@n28) {constB=2}) {constB=0.25})
  NamedRerouteDeclaration(None=Multiply(A=AppendVector(A=@n25, B=@n26.SpriteFall), B=Multiply(A=@n27, B=@n29))) {name=Downwards Movement, nodeColor=<rgba 0 0 0 1>}
  n30 = ScalarParameter() {defaultValue=1, group=Movement, parameterName=Curl Amount}
  n31 = Multiply(A=@n30) {constA=0.65, constB=0.5}
  n32 = PerInstanceRandom()
  n33 = WorldPosition()
  n34 = ObjectPositionWS()
  n35 = ScalarParameter() {defaultValue=50, group=Movement, parameterName=Curl Speed}
  n36 = LinearInterpolate(A=@n35, B=Multiply(A=@n35) {constA=0.65, constB=0.65}, Alpha=PerInstanceRandom())
  n37 = Time()
  n38 = ScalarParameter() {defaultValue=750, group=Movement, parameterName=Curl Scale}
  n39 = Divide(
   A=Subtract(A=Subtract(A=@n33.XYZ, B=@n34), B=Multiply(A=@n36, B=@n37))
   B=LinearInterpolate(A=@n38, B=Multiply(A=@n38) {constA=0.65, constB=0.5}, Alpha=PerInstanceRandom())
  )
  NamedRerouteDeclaration(None=Multiply(A=LinearInterpolate(A=@n30, B=@n31, Alpha=@n32), B=VectorNoise(World Position=@n39) {noiseFunction=VNF_VectorALU})) {name=FN Curl noise Setup, nodeColor=<rgba 0 0 0 1>}
  n40 = VertexColor()
  n41 = Time()
  n42 = Multiply(A=Multiply(A=Multiply(A=@n40) {constB=0.5}, B=Add(A=@n41, B=@n40)), B=Constant() {r=0.05})
  n43 = Sine(None=@n42) {period=0.25}
  n44 = Sine(None=@n42) {period=0.5}
  n45 = Sine(None=@n42) {period=0.5}
  NamedRerouteDeclaration(None=Multiply(A=LinearInterpolate(A=@n43, B=@n44, Alpha=@n45), B=Constant() {r=0.5})) {name=Cheap Curl, nodeColor=<rgba 0 0 0 1>}
  n46 = ComponentMask(None=TextureCoordinate()) {g=true}
  n47 = MaterialFunctionCall() {materialFunction=&BoundingBoxBased_0-1_UVW}
  n48 = LinearInterpolate(A=OneMinus(None=@n46), B=@n46, Alpha=SmoothStep(Min=Constant() {r=0.45}, Max=Constant() {r=0.55}, Value=@n47.B) {constMax=0.6, constMin=0.4})
  n49 = Constant() {r=1}
  n50 = Constant() {r=0.75}
  n51 = Subtract(A=ParticlePositionWS(), B=WorldPosition().XYZ)
  n52 = Constant() {r=0.2}
  NamedRerouteDeclaration(None=Multiply(A=LinearInterpolate(A=@n48, B=@n49, Alpha=@n50), B=Multiply(A=@n51, B=@n52))) {name=Stretch Down, nodeColor=<rgba 0 0 0 1>}
  Step(X=@n47.B) {constY=0.5}
 ]
}
