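# Translucent unlit Niagara material for exploding firework trails, with optional sparks switched by vertex colour.
# World position offset thickens ribbons via SplineThicken, turns sparks into animated sprites, pushes geometry out from the object centre and adds gravity droop.
# Opacity combines band and sphere masks with panning noise, eroded along vertex red by particle alpha.
# Particle colour gets hue shifts and boosted hot ends; three dynamic parameter sets drive explosion, thickness, hue and sparks.
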Material /SparksEffects/Effects/Fireworks/Materials/M_Sparks_Firework_Ribbon_Explode
graph {
 properties {
  bUsedWithNiagaraMeshParticles=true
  bUsedWithNiagaraSprites=true
  blendMode=BLEND_Translucent
  pixelDepthOffsetMode=PDOM_Legacy
  shadingModel=MSM_Unlit
  twoSided=true
  usedShadingModels=Unlit
 }
 nodes [
  n1 = Constant() {r=1}
  n2 = Constant()
  n3 = StaticBool() {value=true}
  n4 = TextureCoordinate()
  n5 = Constant2Vector() {g=7.5, r=1}
  n6 = SphereMask(
   A=MaterialFunctionCall(UVs=Multiply(A=@n4, B=@n5), Texture Scale=Constant2Vector() {g=1, r=1.25}) {materialFunction=&ScaleUVsByCenter}.UVs
   B=Constant() {r=0.5}
   Radius=Constant() {r=0.5}
   Hardness=Constant()
  ) {hardnessPercent=0}
  n7 = VertexColor()
  n8 = Constant() {r=0.95}
  n9 = Multiply(
   A=Power(Base=MaterialFunctionCall!(Width=@n1, Sharpness=@n2, Offset, Direction Switch=@n3, Compare, Input Coordinates) {materialFunction=/Engine/Functions/Engine_MaterialFunctions03/Procedurals/GeneratedBand}.Result) {constExponent=10}
   B=Add(A=Power(Base=@n6), B=Step(Y=@n7.R, X=@n8) {constX=0.9})
  )
  n10 = VertexColor()
  n11 = ParticleColor()
  n12 = Saturate(None=Divide(A=Subtract(A=@n10.R, B=@n11.A), B=Constant() {r=0.75}) {constB=0.5})
  n13 = VertexColor()
  n14 = Power(Base=@n13.R) {constExponent=4}
  n15 = Multiply(
   A=TextureCoordinate()
   B=AppendVector(
    A=ScalarParameter() {defaultValue=0.1, parameterName=Trail Texture Tile X}
    B=ScalarParameter() {defaultValue=0.1, parameterName=Trail Texture Tile Y}
   )
  )
  n16 = Multiply(A=Add(A=Time(), B=VertexColor().G), B=ScalarParameter() {defaultValue=1, parameterName=Ribbon Smoke Speed})
  n17 = Add(
   A=@n13.R
   B=TextureSample(UVs=Panner(Coordinate=@n15, Time=@n16) {speedY=-0.1}) {samplerType=SAMPLERTYPE_Normal, texture=/Game/Global/Textures/Default/Noise/64x64Noise}.R
  )
  n18 = Constant() {desc=Texture Subtract, r=0.2}
  n19 = Saturate(None=Divide(A=Subtract(A=@n17, B=@n18) {constB=0.25}, B=Constant() {desc=Texture Sharpness, r=2}) {constB=0.01})
  n20 = Multiply(A=Multiply(A=Saturate(None=@n9), B=@n12), B=Saturate(None=Add(A=@n14, B=@n19) {desc=Adding More Body to Texture}))
  n21 = Subtract(A=VertexColor().R) {constB=0.65}
  n22 = Multiply(
   A=Saturate(None=Divide(A=@n21) {constB=0.1})
   B=SphereMask(A=TextureCoordinate(), B=Constant() {r=0.5}, Radius=Constant() {r=0.5}, Hardness=Constant()) {hardnessPercent=0}
  )
  n23 = StaticSwitchParameter(
   True=LinearInterpolate(
    A=NamedRerouteDeclaration(None=@n20) {name=RibbonAlpha, nodeColor=<rgba 0.25 1 0.25 1>}
    B=Multiply(A=@n12, B=@n22)
    Alpha=NamedRerouteUsage()
   )
   False=NamedRerouteUsage()
  ) {defaultValue=true, parameterName=Use Sparks}
  MaterialFunctionCall(Alpha Threshold=Saturate(None=Multiply(A=@n23) {constB=5}), Random=Constant()) {materialFunction=&DitherTemporalAA}
  n24 = VertexColor()
  n25 = Power(Base=@n24.R) {constExponent=5}
  n26 = NamedRerouteUsage()
  n27 = NamedRerouteUsage()
  n28 = MaterialFunctionCall(Hue Shift Percentage=Add(A=Multiply(A=@n25, B=@n26), B=Multiply(A=@n24.G, B=@n27)), Texture=ParticleColor().RGB) {materialFunction=&HueShift}
  n29 = LinearInterpolate(
   A=@n28.Result
   B=Multiply(A=@n28.Result, B=Constant() {r=75}) {constB=50}
   Alpha=Power(Base=VertexColor().R, Exp=Constant() {r=10}) {constExponent=5}
  )
  n30 = NamedRerouteUsage()
  n31 = StaticSwitchParameter(True=LinearInterpolate(A=@n29, B=Multiply(A=@n29, B=@n30) {constB=0.5}, Alpha=NamedRerouteUsage()), False=@n29) {defaultValue=true, parameterName=Use Sparks}
  MaterialFunctionCall(Emissive=StaticSwitchParameter(True=Multiply(A=@n31) {constB=0.5}, False=@n31) {parameterName=Opacity Mask}) {materialFunction=&MF_FX_EmissiveAdjustments}
  n32 = DynamicParameter() {paramNames [Explode Amount, Gravity Amount, Hue Shift Gradient, Hue Shift Per Trail]}
  NamedRerouteDeclaration(None=@n32.Explode Amount) {name=Explode Amount, nodeColor=<rgba 0.5 0 1 1>}
  NamedRerouteDeclaration(None=@n32.Gravity Amount) {name=Gravity Amount, nodeColor=<rgba 0.5 0 1 1>}
  n33 = DynamicParameter() {
   parameterIndex=2
   defaultValue=<rgba 0 1 0.5 1>
   paramNames [Randomize Ribbon Explode Distance, Sparks Gravity Mult, Sparks Intensity Mult, Sparks Speed Mult]
  }
  NamedRerouteDeclaration(None=Multiply(A=@n33.Sparks Gravity Mult) {constB=-2000}) {name=Sparks Gravity Mult, nodeColor=<rgba 0.3125 0 1 1>}
  NamedRerouteDeclaration(None=@n33.Sparks Intensity Mult) {name=Sparks Intensity Mult, nodeColor=<rgba 0.5 0 1 1>}
  NamedRerouteDeclaration(None=@n33.Sparks Speed Mult) {name=Sparks Speed Mult, nodeColor=<rgba 0.5 0 1 1>}
  NamedRerouteDeclaration(None=@n32.Hue Shift Gradient) {name=Hue Shift Gradient, nodeColor=<rgba 0.5 0 1 1>}
  NamedRerouteDeclaration(None=@n32.Hue Shift Per Trail) {name=Hue Shift Per Trail, nodeColor=<rgba 0.5 0 1 1>}
  n34 = DynamicParameter() {
   parameterIndex=1
   defaultValue=<rgba 0.35 1 1 20>
   paramNames [End Thickness WPO, Center Thickness WPO, Randomize Thickness Amount, Sparks Size]
  }
  NamedRerouteDeclaration(None=@n34.End Thickness WPO) {name=End Thickness WPO, nodeColor=<rgba 0.5 0 1 1>}
  NamedRerouteDeclaration(None=@n34.Center Thickness WPO) {name=Center Thickness WPO, nodeColor=<rgba 0.5 0 1 1>}
  NamedRerouteDeclaration(None=@n34.Randomize Thickness Amount) {name=Randomize Thickness Amount, nodeColor=<rgba 0.5 0 1 1>}
  NamedRerouteDeclaration(None=@n34.Sparks Size) {name=Sparks Size, nodeColor=<rgba 0.5 0 1 1>}
  NamedRerouteDeclaration(None=@n33.Randomize Ribbon Explode Distance) {name=Randomize Ribbon Explode Distance, nodeColor=<rgba 0.5 0 1 1>}
  n35 = Constant() {r=1}
  NamedRerouteDeclaration(None=If(A=VertexColor().B, "A > B"=@n35, "A == B"=@n35, "A < B"=Constant()) {constB=0.01}) {name=Sparks Mask, nodeColor=<rgba 0 0 0 1>}
  n36 = Constant() {r=-150}
  n37 = Multiply(A=@n36, B=NamedRerouteUsage())
  n38 = Clamp(None=VertexColor().G, Min=OneMinus(None=NamedRerouteUsage())) {minDefault=0.25}
  n39 = Divide(A=Subtract(A=VertexColor().R, B=ParticleColor().A), B=Constant() {r=0.1}) {constB=0.5}
  n40 = NamedRerouteUsage()
  n41 = MaterialFunctionCall(WidthBase=Multiply(A=Saturate(None=@n39), B=Multiply(A=@n37, B=@n38)), WidthTip=Multiply(A=Multiply(A=@n36, B=@n40), B=@n38)) {materialFunction=&SplineThicken}
  n42 = Constant()
  n43 = NamedRerouteUsage()
  n44 = VertexColor()
  n45 = Time()
  n46 = NamedRerouteUsage()
  n47 = Frac(None=Add(A=@n44.B, B=Multiply(A=@n45, B=@n46)))
  n48 = Multiply(
   A=Multiply(
    A=MaterialFunctionCall(R=@n42, G=@n42, B=@n43) {materialFunction=&AppendMany}.RGB
    B=OneMinus(None=@n44.R)
   )
   B=@n47
  ) {constB=400}
  n49 = MaterialFunctionCall!(
   Center Location=WorldPosition().XYZ
   0-1 UVs
   XY scale=Multiply(A=OneMinus(None=@n47), B=NamedRerouteUsage())
   OverridePivot
   "Normals (Optional)"
  ) {materialFunction=&Sprite}
  n50 = StaticSwitchParameter(
   True=LinearInterpolate(A=@n41.WorldPositionOffset, B=Add(A=@n48, B=@n49.World position offset), Alpha=NamedRerouteUsage()) {desc=Spite / Ribbon Lerp}
   False=@n41.WorldPositionOffset
  ) {defaultValue=true, parameterName=Use Sparks}
  n51 = NamedRerouteUsage()
  n52 = VertexColor()
  n53 = Clamp(None=Multiply(A=Add(A=@n52.G)) {constB=0.6}) {maxDefault=2, minDefault=1}
  n54 = Constant() {r=10}
  n55 = LinearInterpolate(A=OneMinus(None=NamedRerouteUsage()), Alpha=VertexColor().G) {constA=0.5}
  n56 = Multiply(
   A=Subtract(A=WorldPosition().XYZ, B=ObjectPositionWS())
   B=MaterialFunctionCall(
    Input=Power(Base=@n51, Exp=@n53)
    Input Low=Constant()
    Input High=Constant() {r=1}
    Target Low=Constant() {r=-1}
    Target High=Multiply(A=@n54, B=@n55)
   ) {materialFunction=&RemapValueRange}.Result
  )
  n57 = Constant()
  n58 = VertexColor()
  n59 = AppendVector(
   A=AppendVector(A=@n57, B=@n57)
   B=Multiply(A=NamedRerouteUsage(), B=LinearInterpolate(Alpha=@n58.G) {constA=-125, constB=-250}) {constB=-100}
  )
  n60 = Power(Base=VertexColor().R, Exp=Constant() {r=4})
  Reroute(None=Add(A=Add(A=@n50, B=@n56), B=Multiply(A=@n59, B=@n60)))
 ]
}
